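# Conversation database for the DoubleDuty NPC, who speaks as Aang, giving the greeting shown when a player talks to them.
# One entry point leads to a speech node with a default line and contextual messages gated by requirements and priority curves.
# These cover a first-ever meeting, reactions to three specific player outfits, and lines conditioned on match conversation count.
# The speech then links out to the shared NPC Library quests and basic services conversations.
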ConversationDatabase /NPCLibrary/NPCs/DoubleDuty/NPCLibrary_Conversation_DoubleDuty
entry_points [{tag=Conversation.Entry.DoubleDuty, destinations [#1]}]
nodes [
 ConversationEntryPointNode_7 {
  id=#1
  class=ConversationEntryPointNode
  outputs [#2]
  properties {entryTag {tagName=Conversation.Entry.DoubleDuty}}
 }
 Speech_C_4 {
  id=#2
  class=Speech_C
  outputs [#3, #4]
  properties {
   nodeName=""
   generalConfig {
    defaultMessage=Ever been penguin sledding?
    contextualMessages [
     {
      message="Hi, I'm Aang!"
      contextRequirements [{requirement {class=FortControllerRequirement_IsFirstEverConversationWithNPC}}]
      priority {curve {rowName=Tandem.Speech.Priority.FirstEverInteraction}}
     }
     {
      message="Do you think, if things were different, we could have been friends?"
      contextRequirements [{requirement {class=FortControllerRequirement_HasCID, properties {allowedCIds [/BRCosmetics/Athena/Items/Cosmetics/Characters/Character_TikiTorch]}}}]
      priority {curve {rowName=Tandem.Speech.Priority.OutfitReaction}}
     }
     {
      message="Katara! Have you seen Momo?"
      contextRequirements [{requirement {class=FortControllerRequirement_HasCID, properties {allowedCIds [/BRCosmetics/Athena/Items/Cosmetics/Characters/Character_DirectContact]}}}]
      priority {curve {rowName=Tandem.Speech.Priority.OutfitReaction}}
     }
     {
      message="Good to see you, Toph! Or, um, well, you know."
      contextRequirements [{requirement {class=FortControllerRequirement_HasCID, properties {allowedCIds [/BRCosmetics/Athena/Items/Cosmetics/Characters/Character_RemoteControl]}}}]
      priority {curve {rowName=Tandem.Speech.Priority.OutfitReaction}}
     }
     {
      message=Ever been penguin sledding?
      contextRequirements [{requirement {class=FortControllerRequirement_CheckNPCMatchConversationCount, properties {nPCUniqueIdTag {tagName=AISpawnerData.Type.Tandem.DoubleDuty}}}}]
      priority {curve {rowName=None}}
     }
     {
      message=I never asked to be the Avatar...
      contextRequirements [{requirement {class=FortControllerRequirement_CheckNPCMatchConversationCount, properties {nPCUniqueIdTag {tagName=AISpawnerData.Type.Tandem.DoubleDuty}}}}]
      priority {curve {rowName=None}}
     }
     {
      message="I laugh at gravity all the time. Haha, gravity."
      contextRequirements [{requirement {class=FortControllerRequirement_CheckNPCMatchConversationCount, properties {nPCUniqueIdTag {tagName=AISpawnerData.Type.Tandem.DoubleDuty}}}}]
      priority {curve {rowName=None}}
     }
     {
      message="It's not magic, it's airbending."
      contextRequirements [{requirement {class=FortControllerRequirement_CheckNPCMatchConversationCount, properties {nPCUniqueIdTag {tagName=AISpawnerData.Type.Tandem.DoubleDuty}}}}]
      priority {curve {rowName=None}}
     }
    ]
   }
  }
 }
 ConversationLinkNode_0 {id=#4, class=ConversationLinkNode, properties {remoteEntryTag {tagName=Conversation.Entry.BasicServices}}}
 ConversationLinkNode_1 {id=#3, class=ConversationLinkNode, properties {remoteEntryTag {tagName=Conversation.Entry.NPCLibrary.Quests}}}
]
